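; Llama puzzle quest actor that shows feedback for a collected jigsaw piece or the completed puzzle, chosen by CurrentLetter.
; BeginPlay plays a collect or completion sound and unhides piece meshes; a timeline chain pops in the selected piece, spawns its emitter and ramps emissive on per-piece dynamic materials.
; FadeOutLetters adds a spin, solved-material swap and celebration effects for the final piece, then shrinks piece 01 and destroys the actor.
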(event EventBeginPlay
  (bind _currentletter (Variables|Default|GetCurrentLetter))
  (bind _sm_piece_07 (Variables|Default|GetSM_Piece_07))
  (bind _sm_piece_06 (Variables|Default|GetSM_Piece_06))
  (bind _sm_piece_05 (Variables|Default|GetSM_Piece_05))
  (bind _sm_piece_04 (Variables|Default|GetSM_Piece_04))
  (bind _sm_piece_03 (Variables|Default|GetSM_Piece_03))
  (bind _sm_piece_02 (Variables|Default|GetSM_Piece_02))
  (bind _sm_piece_01 (Variables|Default|GetSM_Piece_01))
  (bind _returnvalue (Utilities|Select :Index _currentletter :Option%201 _sm_piece_01 :Option%202 _sm_piece_02 :Option%203 _sm_piece_03 :Option%204 _sm_piece_04 :Option%205 _sm_piece_05 :Option%206 _sm_piece_06 :Option%207 _sm_piece_07))
  (Audio|PlaySoundatLocation (select (== (Variables|Default|GetCurrentLetter) 7) (Variables|Default|GetJigsaw%20Complete%20Sound) (Variables|Default|GetJigsaw%20Piece%20Collect%20Sound)) (Transformation|GetWorldLocation (Variables|Default|GetDefaultSceneRoot)))
  (switch int (Variables|Default|GetCurrentLetter)
    (:0
      _)
    (:1
      _)
    (:2
      (Development|SetHiddeninGame (Variables|Default|GetSM_Piece_01) false false)
      (Transformation|SetRelativeLocationAndRotation (Variables|Default|GetSM_Piece_01) "0, 0,50.000000" "0.000000,90.000000,180.000000" false false)
      (bind (_direction _slidetime) (Timeline|Pop%20In%20Puzzle PlayFromStart)
        (:Update
          (Transformation|SetRelativeScale3D _returnvalue (Math|Conversions|ToVector(Float) (Math|Float|Lerp 0.0001 1.0 _slidetime))))
        (:Finished
          ;[1 1 1] TIM: Set These Emitters for each piece to pulse over 1s
          (Effects|SpawnEmitterAttached :EmitterTemplate (Utilities|Select :Index (Variables|Default|GetCurrentLetter) :Option%201 (Variables|Default|GetEmitterPiece1) :Option%202 (Variables|Default|GetEmitterPiece2) :Option%203 (Variables|Default|GetEmitterPiece3) :Option%204 (Variables|Default|GetEmitterPiece5) :Option%205 (Variables|Default|GetEmitterPiece5) :Option%206 (Variables|Default|GetEmitterPiece6) :Option%207 (Variables|Default|GetEmitterPiece7)) :AttachToComponent _returnvalue :AttachPointName "FX_Spawn" :Scale "1.000000,1.000000,1.000000" :LocationType "SnapToTarget" :bAutoDestroy true :PoolingMethod "None" :bAutoActivate true)
          (bind (_direction_2 _rampemissive) (Timeline|Timeline_0 Play)
            (:Update
              (Rendering|Material|SetVectorParameterValue (Utilities|Select :Index (Variables|Default|GetCurrentLetter) :Option%201 (Variables|Default|GetPiece1MID) :Option%202 (Variables|Default|GetPiece2MID) :Option%203 (Variables|Default|GetPiece3MID) :Option%204 (Variables|Default|GetPiece4MID) :Option%205 (Variables|Default|GetPiece5MID) :Option%206 (Variables|Default|GetPiece6MID) :Option%207 (Variables|Default|GetPiece7MID)) "Emissive Multiplier" (Math|Color|Lerp(LinearColor) :B (Variables|Default|GetPopColor) :Alpha _rampemissive) :!class /Script/Engine.MaterialInstanceDynamic))
            (:Finished
              (CallFunction|FadeOutLetters))))))
    (:3
      (Development|SetHiddeninGame (Variables|Default|GetSM_Piece_01) false false))
    (:4
      (Development|SetHiddeninGame (Variables|Default|GetSM_Piece_01) false false))
    (:5
      (Development|SetHiddeninGame (Variables|Default|GetSM_Piece_01) false false))
    (:6
      (Development|SetHiddeninGame (Variables|Default|GetSM_Piece_01) false false))
    (:7
      (Development|SetHiddeninGame (Variables|Default|GetSM_Piece_01) false false)
      (Variables|Default|SetFinalPiece true)))
  (Development|SetHiddeninGame (Variables|Default|GetSM_Piece_01) false false))

(event Custom|FadeOutLetters
  (bind _sm_piece_01 (Variables|Default|GetSM_Piece_01))
  (if (Variables|Default|GetFinalPiece)
    (bind (_direction _spintime _growtime) (Timeline|Spin%20and%20Reveal PlayFromStart)
      (:Update
        (Transformation|SetRelativeRotation _sm_piece_01 (Math|Rotator|MakeRotator 0.0 (Math|Float|Lerp 1.8e+02 1.44e+03 _spintime) -9e+01) false false)
        (Transformation|SetRelativeScale3D _sm_piece_01 (Math|Conversions|ToVector(Float) (Math|Float|Lerp 1.0 1.2 _growtime))))
      (:SuccessFX
        (Rendering|Material|SetMaterial _sm_piece_01 0 "/Figment_S05/Athena/Items/QuestInteractables/Llama_Puzzle/Materials/M_Llama_Puzzle.M_Llama_Puzzle")
        (Audio|PlaySoundatLocation (Variables|Default|GetCelebrationSound) (Transformation|GetWorldLocation (Variables|Default|GetDefaultSceneRoot)))
        (Components|Activation|Activate (Variables|Default|GetP_Celebrate1) false)
        (Utilities|FlowControl|Delay 0.15)
        (Components|Activation|Activate (Variables|Default|GetP_Celebrate) false)
        (Utilities|FlowControl|Delay 2.0))))
  (bind (_direction_2 _fadeouttime) (Timeline|FadeOut PlayFromStart)
    (:Update
      (Transformation|SetRelativeScale3D (Variables|Default|GetSM_Piece_01) (Math|Conversions|ToVector(Float) (Math|Float|Lerp 1.0 0.001 _fadeouttime))))
    (:Finished
      (Actor|DestroyActor))))

(fn SetUpDynamicMats)

(fn ConstructionScript
  (bind _returnvalue (Rendering|Material|CreateDynamicMaterialInstance (Variables|Default|GetSM_Piece_01) 0 "/Game/Athena/Items/QuestInteractables/Llama_Puzzle/Materials/M_Llama_Puzzle_Unsolved.M_Llama_Puzzle_Unsolved" "None" :!class /Script/Engine.PrimitiveComponent))
  (Variables|Default|SetPiece1MID _returnvalue)
  (bind _returnvalue_2 (Rendering|Material|CreateDynamicMaterialInstance (Variables|Default|GetSM_Piece_02) 0 "/Game/Athena/Items/QuestInteractables/Llama_Puzzle/Materials/M_Llama_Puzzle_Unsolved.M_Llama_Puzzle_Unsolved" "None" :!class /Script/Engine.PrimitiveComponent))
  (Variables|Default|SetPiece2MID _returnvalue_2)
  (bind _returnvalue_3 (Rendering|Material|CreateDynamicMaterialInstance (Variables|Default|GetSM_Piece_03) 0 "/Game/Athena/Items/QuestInteractables/Llama_Puzzle/Materials/M_Llama_Puzzle_Unsolved.M_Llama_Puzzle_Unsolved" "None" :!class /Script/Engine.PrimitiveComponent))
  (Variables|Default|SetPiece3MID _returnvalue_3)
  (bind _returnvalue_4 (Rendering|Material|CreateDynamicMaterialInstance (Variables|Default|GetSM_Piece_04) 0 "/Game/Athena/Items/QuestInteractables/Llama_Puzzle/Materials/M_Llama_Puzzle_Unsolved.M_Llama_Puzzle_Unsolved" "None" :!class /Script/Engine.PrimitiveComponent))
  (Variables|Default|SetPiece4MID _returnvalue_4)
  (bind _returnvalue_5 (Rendering|Material|CreateDynamicMaterialInstance (Variables|Default|GetSM_Piece_05) 0 "/Game/Athena/Items/QuestInteractables/Llama_Puzzle/Materials/M_Llama_Puzzle_Unsolved.M_Llama_Puzzle_Unsolved" "None" :!class /Script/Engine.PrimitiveComponent))
  (Variables|Default|SetPiece5MID _returnvalue_5)
  (bind _returnvalue_6 (Rendering|Material|CreateDynamicMaterialInstance (Variables|Default|GetSM_Piece_06) 0 "/Game/Athena/Items/QuestInteractables/Llama_Puzzle/Materials/M_Llama_Puzzle_Unsolved.M_Llama_Puzzle_Unsolved" "None" :!class /Script/Engine.PrimitiveComponent))
  (Variables|Default|SetPiece6MID _returnvalue_6)
  (bind _returnvalue_7 (Rendering|Material|CreateDynamicMaterialInstance (Variables|Default|GetSM_Piece_07) 0 "/Game/Athena/Items/QuestInteractables/Llama_Puzzle/Materials/M_Llama_Puzzle_Unsolved.M_Llama_Puzzle_Unsolved" "None" :!class /Script/Engine.PrimitiveComponent))
  (Variables|Default|SetPiece7MID _returnvalue_7))
