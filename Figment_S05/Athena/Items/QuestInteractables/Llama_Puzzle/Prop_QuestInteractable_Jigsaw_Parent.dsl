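; Parent Blueprint for Llama jigsaw-puzzle quest pickups, showing the piece icon mesh for the player's next uncollected objective and hiding once seven are done.
; Entering the overlap volume starts a looping timer replaying a Rattle timeline that shakes the piece, drives sparkles and material glow, and plays a sound; leaving resets it.
; Completing the objective spawns a completed-puzzle actor facing the interacting pawn, passing the completed count, and unbinds the overlap events.
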(event EventBeginPlay
  (|CallParentFunction) ; This node is disabled and will not be called.\nDrag off pins to build functionality.
)

(event EventTick
  (params (DeltaSeconds :type Float%20%28single-precision%29))
  (|CallParentFunction DeltaSeconds) ; This node is disabled and will not be called.\nDrag off pins to build functionality.
)

(event EventQuestInitializedAndValid
  (bind _startspinoverlap (Variables|Default|GetStartSpinOverlap))
  ;[1 1 1] Initialize Puzzle BP
  (Collision|BindEventtoOnComponentBeginOverlap _startspinoverlap (delegate Custom|PawnCloseEnoughToTriggerBobbing))
  (Collision|BindEventtoOnComponentEndOverlap _startspinoverlap (delegate Custom|PawnOutOfRange))
  (CallFunction|SetLetterAtBeginning)
  (Development|PrintString "Start"))

(event Custom|PawnCloseEnoughToTriggerBobbing
  (params (OverlappedComponent :type Primitive%20Component%20Object%20Reference) (OtherActor :type Actor%20Object%20Reference) (OtherComp :type Primitive%20Component%20Object%20Reference) (OtherBodyIndex :type Integer) (bFromSweep :type Boolean) (SweepResult :type Hit%20Result%20Structure%20%28by%20ref%29))
  (bind _p_quest_puzzle_shake (Variables|Default|GetP_Quest_Puzzle_Shake))
  ;[1 1 1] Enable Rattling Piece
  (bind _returnvalue (Utilities|Time|SetTimerbyEvent (delegate Custom|StartRattle) 4.0 true))
  (Variables|Default|SetRattleTimer _returnvalue)
  (bind (_direction _raise _rattle _rattleglow) (Timeline|Rattle PlayFromStart)
    (:Update
      (Transformation|SetRelativeLocationAndRotation (Variables|Default|GetSM_PuzzlePiece) (Math|Vector|Lerp(Vector) "0, 0,60.000000" "0, 0,70.000000" _raise) (Math|Rotator|Lerp(Rotator) (Math|Rotator|MakeRotator -9e+01) (Math|Rotator|MakeRotator -8e+01 1e+01 1e+01) _rattle true) false false)
      ;[1 1 1] spawnSomeSparklesOnShake
      (Components|Activation|Activate _p_quest_puzzle_shake false)
      (Effects|Components|ParticleSystem|SetFloatParameter _p_quest_puzzle_shake "SpawnRateScale" (Math|Float|Lerp 0.0 1.0 _rattle))
      (Rendering|Material|SetVectorParameterValue (Variables|Default|GetPieceMID) "Emissive Multiplier" (Math|Color|Lerp(LinearColor) :B (Variables|Default|GetPopColor) :Alpha _rattleglow) :!class /Script/Engine.MaterialInstanceDynamic)
      ;[1 1 1] Play Sound for each piece rattle once
      (Utilities|FlowControl|DoOnce false
        (:Completed
          (if (Variables|Default|GetIs%20Piece%20Still%20Visible)
            (Audio|PlaySoundatLocation "/Figment_S05/Athena/Sounds/Quests/Jigsaw/Quest_Jigsaw_AmbRattle_Cue.Quest_Jigsaw_AmbRattle_Cue" (Transformation|GetWorldLocation (Variables|Default|GetSM_PuzzlePiece)))))))))

(event Custom|Reset%20Rattle%20Sound
  ;[1 1 1] Play Sound for each piece rattle once
  (Utilities|FlowControl|DoOnce false
    (:Completed
      (if (Variables|Default|GetIs%20Piece%20Still%20Visible)
        (Audio|PlaySoundatLocation "/Figment_S05/Athena/Sounds/Quests/Jigsaw/Quest_Jigsaw_AmbRattle_Cue.Quest_Jigsaw_AmbRattle_Cue" (Transformation|GetWorldLocation (Variables|Default|GetSM_PuzzlePiece)))))))

(event Custom|StartRattle
  (bind _p_quest_puzzle_shake (Variables|Default|GetP_Quest_Puzzle_Shake))
  ;[1 1 1] Enable Rattling Piece
  (CallFunction|ResetRattleSound)
  (Timeline|Rattle PlayFromStart))

(event Custom|PawnOutOfRange
  (params (OverlappedComponent :type Primitive%20Component%20Object%20Reference) (OtherActor :type Actor%20Object%20Reference) (OtherComp :type Primitive%20Component%20Object%20Reference) (OtherBodyIndex :type Integer))
  (bind _p_quest_puzzle_shake (Variables|Default|GetP_Quest_Puzzle_Shake))
  ;[1 1 1] Disable Rattling Piece
  (Utilities|FlowControl|Sequence
    (:then_0
      ;[1 1 1] Enable Rattling Piece
      (Utilities|Time|ClearandInvalidateTimerbyHandle (Variables|Default|GetRattleTimer))
      (Timeline|Rattle Stop))
    (:then_1
      (Transformation|SetRelativeLocationAndRotation (Variables|Default|GetSM_PuzzlePiece) "0, 0,60.000000" "0, 0,-90.000000" false false))))

(event EventThisQuestObjectiveComplete
  (params (IneractingPawn :type Fort%20Pawn%20Object%20Reference))
  (bind _startspinoverlap (Variables|Default|GetStartSpinOverlap))
  ;[1 1 1] Spawn Puzzle up to the number of pieces you have collected so far
  (bind _returnvalue (Quest|GetNumCompletedObjectives (Quests|GetQuestManager (Utilities|Array|Get(acopy) (Fortnite|GetLocalFortPlayerControllers true)) "Invalid") (Variables|QuestStuff|GetQuestItem)))
  (Development|PrintString (Utilities|String|Append "OBJs Complete: " (Utilities|String|ToString(Integer) _returnvalue)) false false)
  (Game|SpawnActorfromClass "/Figment_S05/Athena/Items/QuestInteractables/Llama_Puzzle/BA_JigsawPuzzle_Complete.BA_JigsawPuzzle_Complete_C" (Math|Transform|MakeTransform (+ (Transformation|GetActorLocation) "0, 0,120.000000") (Math|Conversions|RotationFromXVector (- (* (Transformation|GetActorLocation IneractingPawn) "1.000000,1.000000,0.000000") (* (Transformation|GetActorLocation) "1.000000,1.000000,0.000000")))) "AlwaysSpawn" "OverrideRootScale" 0 _returnvalue 0 0)
  ;[1 1 1] Unbind Rattling Events
  (Collision|UnbindEventfromOnComponentBeginOverlap _startspinoverlap (EventDispatchers|CreateEvent :!event Custom|PawnCloseEnoughToTriggerBobbing))
  (Collision|UnbindEventfromOnComponentEndOverlap _startspinoverlap (EventDispatchers|CreateEvent :!event Custom|PawnOutOfRange))
  (Variables|Default|SetIs%20Piece%20Still%20Visible false))

(event EventQuestUpdatedFromAnywhere
  (bind _sm_puzzlepiece (Variables|Default|GetSM_PuzzlePiece))
  ;[1 1 1] Update Piece to match the next one you are looking for
  (|CallParentFunction)
  (bind _returnvalue (Quest|GetNumCompletedObjectives (Quests|GetQuestManager (Utilities|Array|Get(acopy) (Fortnite|GetLocalFortPlayerControllers true)) "Invalid") (Variables|QuestStuff|GetQuestItem)))
  (switch int _returnvalue
    (:1
      (Components|StaticMesh|SetStaticMesh _sm_puzzlepiece "/Figment_S05/Athena/Items/QuestInteractables/Llama_Puzzle/Meshes/Icons/SM_Llama_Puzzle_Piece_Icon_02.SM_Llama_Puzzle_Piece_Icon_02"))
    (:2
      (Components|StaticMesh|SetStaticMesh _sm_puzzlepiece "/Figment_S05/Athena/Items/QuestInteractables/Llama_Puzzle/Meshes/Icons/SM_Llama_Puzzle_Piece_Icon_03.SM_Llama_Puzzle_Piece_Icon_03"))
    (:3
      (Components|StaticMesh|SetStaticMesh _sm_puzzlepiece "/Figment_S05/Athena/Items/QuestInteractables/Llama_Puzzle/Meshes/Icons/SM_Llama_Puzzle_Piece_Icon_04.SM_Llama_Puzzle_Piece_Icon_04"))
    (:4
      (Components|StaticMesh|SetStaticMesh _sm_puzzlepiece "/Figment_S05/Athena/Items/QuestInteractables/Llama_Puzzle/Meshes/Icons/SM_Llama_Puzzle_Piece_Icon_05.SM_Llama_Puzzle_Piece_Icon_05"))
    (:5
      (Components|StaticMesh|SetStaticMesh _sm_puzzlepiece "/Figment_S05/Athena/Items/QuestInteractables/Llama_Puzzle/Meshes/Icons/SM_Llama_Puzzle_Piece_Icon_06.SM_Llama_Puzzle_Piece_Icon_06"))
    (:6
      (Components|StaticMesh|SetStaticMesh _sm_puzzlepiece "/Game/Athena/Items/QuestInteractables/Llama_Puzzle/Meshes/Icons/SM_Llama_Puzzle_Piece_Icon_07.SM_Llama_Puzzle_Piece_Icon_07"))
    (:7
      (Collision|SetActorEnableCollision false)
      (CallFunction|HideQuestProp))))

(event Custom|Set%20Letter%20at%20Beginning
  (bind _sm_puzzlepiece (Variables|Default|GetSM_PuzzlePiece))
  ;[1 1 1] Update Piece to match the next one you are looking for
  (bind _returnvalue (Quest|GetNumCompletedObjectives (Quests|GetQuestManager (Utilities|Array|Get(acopy) (Fortnite|GetLocalFortPlayerControllers true)) "Invalid") (Variables|QuestStuff|GetQuestItem)))
  (switch int _returnvalue
    (:1
      (Components|StaticMesh|SetStaticMesh _sm_puzzlepiece "/Figment_S05/Athena/Items/QuestInteractables/Llama_Puzzle/Meshes/Icons/SM_Llama_Puzzle_Piece_Icon_02.SM_Llama_Puzzle_Piece_Icon_02"))
    (:2
      (Components|StaticMesh|SetStaticMesh _sm_puzzlepiece "/Figment_S05/Athena/Items/QuestInteractables/Llama_Puzzle/Meshes/Icons/SM_Llama_Puzzle_Piece_Icon_03.SM_Llama_Puzzle_Piece_Icon_03"))
    (:3
      (Components|StaticMesh|SetStaticMesh _sm_puzzlepiece "/Figment_S05/Athena/Items/QuestInteractables/Llama_Puzzle/Meshes/Icons/SM_Llama_Puzzle_Piece_Icon_04.SM_Llama_Puzzle_Piece_Icon_04"))
    (:4
      (Components|StaticMesh|SetStaticMesh _sm_puzzlepiece "/Figment_S05/Athena/Items/QuestInteractables/Llama_Puzzle/Meshes/Icons/SM_Llama_Puzzle_Piece_Icon_05.SM_Llama_Puzzle_Piece_Icon_05"))
    (:5
      (Components|StaticMesh|SetStaticMesh _sm_puzzlepiece "/Figment_S05/Athena/Items/QuestInteractables/Llama_Puzzle/Meshes/Icons/SM_Llama_Puzzle_Piece_Icon_06.SM_Llama_Puzzle_Piece_Icon_06"))
    (:6
      (Components|StaticMesh|SetStaticMesh _sm_puzzlepiece "/Game/Athena/Items/QuestInteractables/Llama_Puzzle/Meshes/Icons/SM_Llama_Puzzle_Piece_Icon_07.SM_Llama_Puzzle_Piece_Icon_07"))
    (:7
      (Collision|SetActorEnableCollision false)
      (CallFunction|HideQuestProp))))

(fn PieceRattle)

(fn ConstructionScript
  (|CallParentFunction)
  (bind _returnvalue (Rendering|Material|CreateDynamicMaterialInstance (Variables|Default|GetSM_PuzzlePiece) 0 (Rendering|Material|GetMaterial (Variables|Default|GetSM_PuzzlePiece) 0) "None" :!class /Script/Engine.PrimitiveComponent))
  (Variables|Default|SetPieceMID _returnvalue))
